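# Physics asset for the F_MED_Bullet_Blue backpack cosmetic, giving collision and jointed motion to its dangling chain, straps and shell pieces.
# The pack body, a capsule plus a box, is the hub, with master bodies for the chain, each strap and each shell group constrained to it with free swing and twist.
# Two-link chain and strap series and three shells per side hang from those masters on angular limits, with near-zero twist on straps.
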PhysicsAsset /BRCosmetics/Accessories/FORT_Backpacks/F_MED_Bullet_Blue/Meshes/F_MED_Bullet_Blue_Pack_PhysicsAsset
bodies {
 dyn_chain_01 [{shapeType=Box, extentX=6.132, extentY=6.43, extentZ=2.1729, center=<2.7904 0 0>}]
 dyn_chain_02 [{shapeType=Box, extentX=6.9216, extentY=6.2743, extentZ=1.6304, center=<3.3733 0 0>}]
 dyn_chain_mstr [
  {shapeType=Box, extentX=1.01, extentY=1.01, extentZ=1.01}
  {shapeType=Box, extentX=8.9124, extentY=1.01, extentZ=5.2338, center=<0.1069 -4.0887 -6.748>, rotation=<rot 0 0 38.0209>}
 ]
 dyn_strap_01_l [{shapeType=Box, extentX=5.0402, extentY=7.7715, extentZ=0.9313, center=<-2 0 0>}]
 dyn_strap_01_r [{shapeType=Box, extentX=5.6101, extentY=7.7715, extentZ=0.8152, center=<2 0 0>}]
 dyn_strap_02_l [{shapeType=Box, extentX=5.4219, extentY=4.1281, extentZ=3.4439, center=<-2 0 0>}]
 dyn_strap_02_r [{shapeType=Box, extentX=5.2139, extentY=4.1683, extentZ=3.4818, center=<2 0 0>}]
 dyn_strap_l_mstr [{shapeType=Box, extentX=1.01, extentY=1.01, extentZ=1.01}]
 dyn_strap_r_mstr [{shapeType=Box, extentX=1.01, extentY=1.01, extentZ=1.01}]
 pack [
  {shapeType=Capsule, length=8.9816, radius=10.3495, center=<0.507 2.6134 -8.5456>, rotation=<rot 9.4089 -89.8382 -178.8236>}
  {shapeType=Box, extentX=18.6816, extentY=20.1824, extentZ=4.665, center=<0.5996 8.9079 -14.6667>, rotation=<rot 0 91.0861 0>}
 ]
 shell_a_l [{shapeType=Box, extentX=5.8559, extentY=3.088, extentZ=4.7367, center=<0 1.225 0>}]
 shell_a_r [{shapeType=Box, extentX=5.8545, extentY=2.9949, extentZ=3.5931, center=<0.0001 -0.8344 0>}]
 shell_b_l [{shapeType=Box, extentX=5.9129, extentY=2.9697, extentZ=5.4094, center=<0 1.225 0>}]
 shell_b_r [{shapeType=Box, extentX=5.9129, extentY=3.3592, extentZ=4.1269, center=<0.0001 -0.8344 0>}]
 shell_c_l [{shapeType=Box, extentX=5.8864, extentY=3.1426, extentZ=4.1578, center=<0 1.225 0>}]
 shell_c_r [{shapeType=Box, extentX=5.8864, extentY=3.0123, extentZ=3.334, center=<0.0001 -0.8344 0>}]
 shell_mstr_l [{shapeType=Box, extentX=1.01, extentY=1.01, extentZ=1.01}]
 shell_mstr_r [{shapeType=Box, extentX=1.01, extentY=1.01, extentZ=1.01}]
}
constraints [
 {
  bone1Name=dyn_strap_r_mstr
  bone2Name=pack
  swing1LimitDegrees=45
  swing1Motion=Free
  swing2LimitDegrees=45
  swing2Motion=Free
  twistLimitDegrees=45
  twistMotion=Free
 }
 {
  bone1Name=dyn_chain_mstr
  bone2Name=pack
  swing1LimitDegrees=45
  swing1Motion=Free
  swing2LimitDegrees=45
  swing2Motion=Free
  twistLimitDegrees=45
  twistMotion=Free
 }
 {
  bone1Name=dyn_strap_l_mstr
  bone2Name=pack
  swing1LimitDegrees=45
  swing1Motion=Free
  swing2LimitDegrees=45
  swing2Motion=Free
  twistLimitDegrees=45
  twistMotion=Free
 }
 {
  bone1Name=shell_mstr_l
  bone2Name=pack
  swing1LimitDegrees=45
  swing1Motion=Free
  swing2LimitDegrees=45
  swing2Motion=Free
  twistLimitDegrees=45
  twistMotion=Free
 }
 {
  bone1Name=shell_mstr_r
  bone2Name=pack
  swing1LimitDegrees=45
  swing1Motion=Free
  swing2LimitDegrees=45
  swing2Motion=Free
  twistLimitDegrees=45
  twistMotion=Free
 }
 {bone1Name=dyn_strap_02_r, bone2Name=dyn_strap_01_r, swing1LimitDegrees=10, swing2LimitDegrees=60, twistLimitDegrees=0.2}
 {bone1Name=dyn_strap_01_r, bone2Name=dyn_strap_r_mstr, swing1LimitDegrees=10, swing2LimitDegrees=80, twistLimitDegrees=0.2}
 {bone1Name=dyn_chain_02, bone2Name=dyn_chain_01, swing1LimitDegrees=45, swing2LimitDegrees=33, twistLimitDegrees=5}
 {bone1Name=dyn_chain_01, bone2Name=dyn_chain_mstr, swing1LimitDegrees=30, swing2LimitDegrees=33, twistLimitDegrees=10}
 {bone1Name=dyn_strap_01_l, bone2Name=dyn_strap_l_mstr, swing1LimitDegrees=10, swing2LimitDegrees=80, twistLimitDegrees=0.2}
 {bone1Name=dyn_strap_02_l, bone2Name=dyn_strap_01_l, swing1LimitDegrees=10, swing2LimitDegrees=60, twistLimitDegrees=0.2}
 {bone1Name=shell_a_l, bone2Name=shell_mstr_l, swing1LimitDegrees=10, swing2LimitDegrees=2, twistLimitDegrees=10}
 {bone1Name=shell_b_l, bone2Name=shell_mstr_l, swing1LimitDegrees=10, swing2LimitDegrees=2, twistLimitDegrees=10}
 {bone1Name=shell_c_l, bone2Name=shell_mstr_l, swing1LimitDegrees=10, swing2LimitDegrees=2, twistLimitDegrees=10}
 {bone1Name=shell_a_r, bone2Name=shell_mstr_r, swing1LimitDegrees=10, swing2LimitDegrees=2, twistLimitDegrees=10}
 {bone1Name=shell_b_r, bone2Name=shell_mstr_r, swing1LimitDegrees=10, swing2LimitDegrees=2, twistLimitDegrees=10}
 {bone1Name=shell_c_r, bone2Name=shell_mstr_r, swing1LimitDegrees=10, swing2LimitDegrees=2, twistLimitDegrees=10}
]
